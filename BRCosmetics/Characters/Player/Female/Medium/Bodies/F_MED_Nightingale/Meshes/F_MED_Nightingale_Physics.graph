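# Physics asset for the Nightingale medium female player body, setting up skirt dynamics over pelvis and thigh capsules.
# Eight skirt panels each hang with limited swing and twist from a tiny MSTR anchor capsule.
# Anchors join with free rotation to the thighs for sides and outer fronts, to the pelvis for backs and front middle.
# Cross-constraints link neighboring panels, angularly limited across the front and free elsewhere, leaving sides and backs unlinked.
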PhysicsAsset /BRCosmetics/Characters/Player/Female/Medium/Bodies/F_MED_Nightingale/Meshes/F_MED_Nightingale_Physics
bodies {
 dyn_skirt_bk_l [{shapeType=Capsule, length=8, radius=3, shapeName=dyn_skirt_bk_l_capsule, center=<8 -1 0>}]
 dyn_skirt_bk_l_MSTR [{shapeType=Capsule, length=0.505, radius=0.505, shapeName=dyn_skirt_bk_l_MSTR_capsule, rotation=<rot 0 -90 0>}]
 dyn_skirt_bk_mid [{shapeType=Capsule, length=8, radius=3, shapeName=dyn_skirt_bk_mid_capsule, center=<8 -1 0>}]
 dyn_skirt_bk_mid_MSTR [{shapeType=Capsule, length=0.505, radius=0.505, shapeName=dyn_skirt_bk_mid_MSTR_capsule, rotation=<rot 0 -90 0>}]
 dyn_skirt_bk_r [{shapeType=Capsule, length=8, radius=3, shapeName=dyn_skirt_bk_r_capsule, center=<8 -1 0>}]
 dyn_skirt_bk_r_MSTR [{shapeType=Capsule, length=0.505, radius=0.505, shapeName=dyn_skirt_bk_r_MSTR_capsule, rotation=<rot 0 -90 0>}]
 dyn_skirt_fr_l [{shapeType=Capsule, length=10, radius=2, shapeName=dyn_skirt_fr_l_capsule, center=<5 -2.5 0>}]
 dyn_skirt_fr_l_MSTR [{shapeType=Capsule, length=0.505, radius=0.505, shapeName=dyn_skirt_fr_l_MSTR_capsule, rotation=<rot 90 0 0>}]
 dyn_skirt_fr_mid [{shapeType=Capsule, length=10, radius=2, shapeName=dyn_skirt_fr_r_capsule, center=<5 -2.5 0>}]
 dyn_skirt_fr_mid_MSTR [{shapeType=Capsule, length=0.505, radius=0.505, shapeName=dyn_skirt_fr_mid_MSTR_capsule, rotation=<rot 90 0 0>}]
 dyn_skirt_fr_r [{shapeType=Capsule, length=10, radius=2, shapeName=dyn_skirt_fr_r_capsule, center=<5 -2.5 0>}]
 dyn_skirt_fr_r_MSTR [{shapeType=Capsule, length=0.505, radius=0.505, shapeName=dyn_skirt_fr_r_MSTR_capsule, rotation=<rot 90 0 0>}]
 dyn_skirt_l [{shapeType=Capsule, length=12, radius=2.5, shapeName=dyn_skirt_l_capsule, center=<4 -2 0>}]
 dyn_skirt_l_MSTR [{shapeType=Capsule, length=0.505, radius=0.505, shapeName=dyn_skirt_l_MSTR_capsule, rotation=<rot 90 0 0>}]
 dyn_skirt_r [{shapeType=Capsule, length=12, radius=2.5, shapeName=dyn_skirt_r_capsule, center=<4 -2 0>}]
 dyn_skirt_r_MSTR [{shapeType=Capsule, length=0.505, radius=0.505, shapeName=dyn_skirt_r_MSTR_capsule, rotation=<rot 90 0 0>}]
 pelvis [{shapeType=Capsule, length=12.7181, radius=9.75, center=<-6.5 -1.25 0>, rotation=<rot 0 -90 0>}]
 thigh_l [{shapeType=Capsule, length=31, radius=7.75, center=<-23 0.5 0>, rotation=<rot 90 0 0>}]
 thigh_r [{shapeType=Capsule, length=31, radius=7.75, center=<23 -0.5 0>, rotation=<rot 90 0 0>}]
}
constraints [
 {bone1Name=thigh_l, bone2Name=pelvis, swing1LimitDegrees=45, swing2LimitDegrees=45, twistLimitDegrees=45}
 {bone1Name=thigh_r, bone2Name=pelvis, swing1LimitDegrees=45, swing2LimitDegrees=45, twistLimitDegrees=45}
 {
  bone1Name=dyn_skirt_fr_r_MSTR
  bone2Name=thigh_r
  swing1LimitDegrees=45
  swing1Motion=Free
  swing2LimitDegrees=45
  swing2Motion=Free
  twistLimitDegrees=45
  twistMotion=Free
 }
 {bone1Name=dyn_skirt_fr_r, bone2Name=dyn_skirt_fr_r_MSTR, swing1LimitDegrees=45, swing2LimitDegrees=35, twistLimitDegrees=10}
 {
  bone1Name=dyn_skirt_r_MSTR
  bone2Name=thigh_r
  swing1LimitDegrees=45
  swing1Motion=Free
  swing2LimitDegrees=45
  swing2Motion=Free
  twistLimitDegrees=45
  twistMotion=Free
 }
 {bone1Name=dyn_skirt_r, bone2Name=dyn_skirt_r_MSTR, swing1LimitDegrees=30, swing2LimitDegrees=20, twistLimitDegrees=10}
 {bone1Name=dyn_skirt_bk_r, bone2Name=dyn_skirt_bk_r_MSTR, swing1LimitDegrees=40, swing2LimitDegrees=45, twistLimitDegrees=10}
 {
  bone1Name=dyn_skirt_fr_mid_MSTR
  bone2Name=pelvis
  swing1LimitDegrees=45
  swing1Motion=Free
  swing2LimitDegrees=45
  swing2Motion=Free
  twistLimitDegrees=45
  twistMotion=Free
 }
 {bone1Name=dyn_skirt_fr_mid, bone2Name=dyn_skirt_fr_mid_MSTR, swing1LimitDegrees=45, swing2LimitDegrees=35, twistLimitDegrees=10}
 {bone1Name=dyn_skirt_bk_mid, bone2Name=dyn_skirt_bk_mid_MSTR, swing1LimitDegrees=40, swing2LimitDegrees=45, twistLimitDegrees=10}
 {
  bone1Name=dyn_skirt_fr_l_MSTR
  bone2Name=thigh_l
  swing1LimitDegrees=45
  swing1Motion=Free
  swing2LimitDegrees=45
  swing2Motion=Free
  twistLimitDegrees=45
  twistMotion=Free
 }
 {bone1Name=dyn_skirt_fr_l, bone2Name=dyn_skirt_fr_l_MSTR, swing1LimitDegrees=45, swing2LimitDegrees=35, twistLimitDegrees=10}
 {
  bone1Name=dyn_skirt_l_MSTR
  bone2Name=thigh_l
  swing1LimitDegrees=45
  swing1Motion=Free
  swing2LimitDegrees=45
  swing2Motion=Free
  twistLimitDegrees=45
  twistMotion=Free
 }
 {bone1Name=dyn_skirt_l, bone2Name=dyn_skirt_l_MSTR, swing1LimitDegrees=30, swing2LimitDegrees=20, twistLimitDegrees=10}
 {bone1Name=dyn_skirt_bk_l, bone2Name=dyn_skirt_bk_l_MSTR, swing1LimitDegrees=40, swing2LimitDegrees=45, twistLimitDegrees=10}
 {bone1Name=dyn_skirt_fr_mid, bone2Name=dyn_skirt_fr_r, swing1LimitDegrees=30, swing2LimitDegrees=30, twistLimitDegrees=10}
 {bone1Name=dyn_skirt_fr_mid, bone2Name=dyn_skirt_fr_l, swing1LimitDegrees=30, swing2LimitDegrees=30, twistLimitDegrees=10}
 {
  bone1Name=dyn_skirt_bk_mid
  bone2Name=dyn_skirt_bk_r
  swing1LimitDegrees=25
  swing1Motion=Free
  swing2LimitDegrees=25
  swing2Motion=Free
  twistLimitDegrees=10
  twistMotion=Free
 }
 {
  bone1Name=dyn_skirt_bk_mid
  bone2Name=dyn_skirt_bk_l
  swing1LimitDegrees=25
  swing1Motion=Free
  swing2LimitDegrees=25
  swing2Motion=Free
  twistLimitDegrees=10
  twistMotion=Free
 }
 {
  bone1Name=dyn_skirt_bk_l_MSTR
  bone2Name=pelvis
  swing1LimitDegrees=45
  swing1Motion=Free
  swing2LimitDegrees=45
  swing2Motion=Free
  twistLimitDegrees=45
  twistMotion=Free
 }
 {
  bone1Name=dyn_skirt_bk_r_MSTR
  bone2Name=pelvis
  swing1LimitDegrees=45
  swing1Motion=Free
  swing2LimitDegrees=45
  swing2Motion=Free
  twistLimitDegrees=45
  twistMotion=Free
 }
 {
  bone1Name=dyn_skirt_bk_mid_MSTR
  bone2Name=pelvis
  swing1LimitDegrees=45
  swing1Motion=Free
  swing2LimitDegrees=45
  swing2Motion=Free
  twistLimitDegrees=45
  twistMotion=Free
 }
 {
  bone1Name=dyn_skirt_fr_r
  bone2Name=dyn_skirt_r
  swing1LimitDegrees=45
  swing1Motion=Free
  swing2LimitDegrees=45
  swing2Motion=Free
  twistLimitDegrees=45
  twistMotion=Free
 }
 {
  bone1Name=dyn_skirt_fr_l
  bone2Name=dyn_skirt_l
  swing1LimitDegrees=45
  swing1Motion=Free
  swing2LimitDegrees=45
  swing2Motion=Free
  twistLimitDegrees=45
  twistMotion=Free
 }
]
