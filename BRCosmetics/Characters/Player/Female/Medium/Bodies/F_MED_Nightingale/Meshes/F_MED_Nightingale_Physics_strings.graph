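# Physics asset for the F_MED_Nightingale female body cosmetic that defines physics bodies and joints for its left and right hood strings.
# Each string is a chain of three capsules hanging from a small master body, with limited swing and tighter twist between links.
# Both master bodies join spine_05 with free angular motion, and three large capsules on spine_03 to spine_05 approximate the torso.
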PhysicsAsset /BRCosmetics/Characters/Player/Female/Medium/Bodies/F_MED_Nightingale/Meshes/F_MED_Nightingale_Physics_strings
bodies {
 dyn_hoodString_l_01 [{shapeType=Capsule, length=3.75, radius=1, shapeName=dyn_hoodString_l_01_capsule, center=<4 0 0>, rotation=<rot 90 0 0>}]
 dyn_hoodString_l_02 [{shapeType=Capsule, length=4.5, radius=1, shapeName=dyn_hoodString_l_02_capsule, center=<4 0 0>, rotation=<rot 90 0 0>}]
 dyn_hoodString_l_03 [{shapeType=Capsule, length=3.75, radius=1, shapeName=dyn_hoodString_l_03_capsule, center=<4 0 0>, rotation=<rot 90 0 0>}]
 dyn_hoodString_l_MSTR [{shapeType=Capsule, length=0.505, radius=0.505, shapeName=dyn_hoodString_l_MSTR_capsule, rotation=<rot 90 0 0>}]
 dyn_hoodString_r_01 [{shapeType=Capsule, length=4.5, radius=1, shapeName=dyn_hoodString_r_01_capsule, center=<4 0 0>, rotation=<rot 90 0 0>}]
 dyn_hoodString_r_02 [{shapeType=Capsule, length=3.75, radius=1, shapeName=dyn_hoodString_r_02_capsule, center=<4 0 0>, rotation=<rot 90 0 0>}]
 dyn_hoodString_r_03 [{shapeType=Capsule, length=4.5, radius=1, shapeName=dyn_hoodString_r_03_capsule, center=<4 0 0>, rotation=<rot 90 0 0>}]
 dyn_hoodString_r_MSTR [{shapeType=Capsule, length=0.505, radius=0.505, shapeName=dyn_hoodString_r_MSTR_capsule, rotation=<rot 90 0 0>}]
 spine_03 [{shapeType=Capsule, length=6, radius=10.5, center=<0 0.75 0>}]
 spine_04 [{shapeType=Capsule, length=5.125, radius=11, center=<0 0.75 0>}]
 spine_05 [{shapeType=Capsule, length=9.625, radius=10}]
}
constraints [
 {bone1Name=spine_05, bone2Name=spine_04, swing1LimitDegrees=45, swing2LimitDegrees=45, twistLimitDegrees=45}
 {
  bone1Name=dyn_hoodString_l_MSTR
  bone2Name=spine_05
  swing1LimitDegrees=45
  swing1Motion=Free
  swing2LimitDegrees=45
  swing2Motion=Free
  twistLimitDegrees=45
  twistMotion=Free
 }
 {bone1Name=dyn_hoodString_l_01, bone2Name=dyn_hoodString_l_MSTR, swing1LimitDegrees=45, swing2LimitDegrees=45, twistLimitDegrees=10}
 {bone1Name=dyn_hoodString_l_02, bone2Name=dyn_hoodString_l_01, swing1LimitDegrees=45, swing2LimitDegrees=45, twistLimitDegrees=10}
 {bone1Name=dyn_hoodString_l_03, bone2Name=dyn_hoodString_l_02, swing1LimitDegrees=45, swing2LimitDegrees=45, twistLimitDegrees=10}
 {
  bone1Name=dyn_hoodString_r_MSTR
  bone2Name=spine_05
  swing1LimitDegrees=45
  swing1Motion=Free
  swing2LimitDegrees=45
  swing2Motion=Free
  twistLimitDegrees=45
  twistMotion=Free
 }
 {bone1Name=dyn_hoodString_r_01, bone2Name=dyn_hoodString_r_MSTR, swing1LimitDegrees=45, swing2LimitDegrees=45, twistLimitDegrees=10}
 {bone1Name=dyn_hoodString_r_02, bone2Name=dyn_hoodString_r_01, swing1LimitDegrees=45, swing2LimitDegrees=45, twistLimitDegrees=10}
 {bone1Name=dyn_hoodString_r_03, bone2Name=dyn_hoodString_r_02, swing1LimitDegrees=45, swing2LimitDegrees=45, twistLimitDegrees=10}
 {
  bone1Name=spine_04
  bone2Name=spine_03
  swing1LimitDegrees=45
  swing1Motion=Free
  swing2LimitDegrees=45
  swing2Motion=Free
  twistLimitDegrees=45
  twistMotion=Free
 }
]
