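# Cosmetic anim blueprint for the Nightingale female body, adding secondary motion to skirt, hoodie and straps.
# It copies the parent mesh pose, centers mid-skirt bones and offsets skirt bones by thigh-angle alphas from the EventGraph.
# The cached result feeds two RigidBody simulations, AnimDynamics leg straps, additive hoodie adjustments and skirt-to-thigh rotation constraints.
# The EventGraph also computes gravity, skirt-constrain, hoodie and spine values, without visible AnimGraph connections.
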AnimBlueprint /BRCosmetics/Characters/Player/Female/Medium/Bodies/F_MED_Nightingale/Meshes/F_MED_Nightingale_AnimBP
target_skeleton=&/F_MED_Nightingale_Skeleton
graphs [
 AnimGraph {nodes [
   n1 = RigidBody!(ComponentPose=Local To Component(LocalPose=Use cached pose 'PreSim Cache'()), Alpha, OverrideWorldGravity:Vector="0,0,-980", ExternalForce) -> (Pose) {
    bClampLinearTranslationLimitToRefPose=true
    bOverrideWorldGravity=true
    overridePhysicsAsset=&/F_MED_Nightingale_Physics
    simulationSpace=BaseBoneSpace
    alphaBoolBlend {blendOption=Linear}
    baseBoneRef {boneName=pelvis}
    componentAppliedLinearAccClamp=<1500 1500 1500>
    componentLinearAccScale=<0.2 0.2 0.3>
    componentLinearVelScale=<0.2 0.2 0.3>
    overrideWorldGravity {z=-980}
    simSpaceSettings {
     maxAngularAcceleration=1000
     maxAngularVelocity=3
     maxLinearAcceleration=700
     maxLinearVelocity=500
     velocityScaleZ=0
     worldAlpha=1
     externalLinearDragV=<0.2 0.2>
    }
   }
   n2 = RigidBody(ComponentPose=@n1, OverrideWorldGravity="0,0,-800") {
    bClampLinearTranslationLimitToRefPose=true
    bOverrideWorldGravity=true
    overridePhysicsAsset=&/F_MED_Nightingale_Physics_strings
    simulationSpace=BaseBoneSpace
    alphaBoolBlend {blendOption=Linear}
    baseBoneRef {boneName=spine_04}
    componentAppliedLinearAccClamp=<3000 3000 3000>
    componentLinearAccScale=<0.3 0.3 0.4>
    componentLinearVelScale=<0.2 0.2 0.35>
    overrideWorldGravity {z=-800}
   }
   n3 = Anim Dynamics - Bone: dyn_legStrap_l_01!(ComponentPose=@n2, Alpha=1, GravityOverride:Vector="0,0,-980", ExternalForce="0,0,0") -> (Pose) {
    angularDampingOverride=0.99
    bOverrideAngularDamping=true
    bUseGravityOverride=true
    simulationSpace=RootRelative
    alphaBoolBlend {blendOption=Linear}
    boundBone {boneName=dyn_legStrap_l_01}
    componentAppliedLinearAccClamp=<3000 3000 3000>
    componentLinearAccScale=<0.3 0.3 0.45>
    componentLinearVelScale=<0.3 0.3 0.45>
    gravityOverride {z=-980}
    physicsBodyDefinitions [{
      boundBone {boneName=dyn_legStrap_l_01}
      boxExtents=<30 10 30>
      constraintSetup {angularLimitsMax=<5 30 10>, angularLimitsMin=<-5 -30 -50>, angularTarget=<0 0 0>}
      localJointOffset=<6 0 0>
     }]
   }
   n4 = Anim Dynamics - Bone: dyn_legStrap_r_01!(ComponentPose=@n3, Alpha=1, GravityOverride:Vector="0,0,-980", ExternalForce="0,0,0") -> (Pose) {
    angularDampingOverride=0.99
    bOverrideAngularDamping=true
    bUseGravityOverride=true
    simulationSpace=RootRelative
    alphaBoolBlend {blendOption=Linear}
    boundBone {boneName=dyn_legStrap_r_01}
    componentAppliedLinearAccClamp=<3000 3000 3000>
    componentLinearAccScale=<0.3 0.3 0.45>
    componentLinearVelScale=<0.3 0.3 0.45>
    gravityOverride {z=-980}
    physicsBodyDefinitions [{
      boundBone {boneName=dyn_legStrap_r_01}
      boxExtents=<30 10 30>
      constraintSetup {angularLimitsMax=<5 30 10>, angularLimitsMin=<-5 -30 -50>, angularTarget=<0 0 0>}
      localJointOffset=<6 0 0>
     }]
   }
   n5 = "Transform (Modify) Bone - Bone: dyn_hoodieLow_adj_MSTR"(ComponentPose=@n4, Alpha=1, Translation="0,1,6", Rotation="0,0,0", Scale="0.8,1.2,0.5") {
    rotationMode=BMM_Additive
    rotationSpace=BCS_BoneSpace
    scaleMode=BMM_Additive
    scaleSpace=BCS_BoneSpace
    translationMode=BMM_Additive
    translationSpace=BCS_BoneSpace
    alphaBoolBlend {blendInTime=0.1, blendOption=Linear, blendOutTime=0.1}
    boneToModify {boneName=dyn_hoodieLow_adj_MSTR}
    scale=<0.8 1.2 0.5>
    translation {y=1, z=6}
   }
   n6 = "Transform (Modify) Bone - Bone: dyn_hoodieLow_adj_MSTR"(ComponentPose=@n5, Alpha=1, Translation="0,0,0", Rotation="30,-25,0", Scale="1,1,1") {
    rotationMode=BMM_Additive
    rotationSpace=BCS_BoneSpace
    scaleMode=BMM_Additive
    scaleSpace=BCS_BoneSpace
    translationMode=BMM_Additive
    translationSpace=BCS_BoneSpace
    alphaBoolBlend {blendInTime=0.1, blendOption=Linear, blendOutTime=0.1}
    boneToModify {boneName=dyn_hoodieLow_adj_MSTR}
    rotation {pitch=30, yaw=-25}
   }
   n7 = "Transform (Modify) Bone - Bone: dyn_hoodieLow_adj_MSTR"(ComponentPose=@n6, Alpha=1, Translation="0,0,0", Rotation="-30,25,0", Scale="1,1,1") {
    rotationMode=BMM_Additive
    rotationSpace=BCS_BoneSpace
    scaleMode=BMM_Additive
    scaleSpace=BCS_BoneSpace
    translationMode=BMM_Additive
    translationSpace=BCS_BoneSpace
    alphaBoolBlend {blendInTime=0.1, blendOption=Linear, blendOutTime=0.1}
    boneToModify {boneName=dyn_hoodieLow_adj_MSTR}
    rotation {pitch=-30, yaw=25}
   }
   n8 = "Transform (Modify) Bone - Bone: dyn_hoodieLow_adj_MSTR"(ComponentPose=@n7, Alpha=1, Translation="0,-5.5,4", Rotation="0,0,0", Scale="1,0.8,1") {
    rotationMode=BMM_Additive
    rotationSpace=BCS_BoneSpace
    scaleMode=BMM_Additive
    scaleSpace=BCS_BoneSpace
    translationMode=BMM_Additive
    translationSpace=BCS_BoneSpace
    alphaBoolBlend {blendInTime=0.1, blendOption=Linear, blendOutTime=0.1}
    boneToModify {boneName=dyn_hoodieLow_adj_MSTR}
    scale {y=0.8}
    translation {y=-5.5, z=4}
   }
   n9 = "Constraint - dyn_skirt_fr_l (R)"!(ComponentPose=@n8, bAlphaBoolEnabled:bool=True, ConstraintWeights_0=1) -> (Pose) {
    alphaInputType=Bool
    alphaBoolBlend {blendInTime=0.025, blendOption=Linear, blendOutTime=0.25}
    boneToModify {boneName=dyn_skirt_fr_l}
    constraintSetup [{transformType=Rotation, targetBone {boneName=thigh_l}}]
    constraintWeights [1]
   }
   n10 = "Constraint - dyn_skirt_fr_r (R)"!(ComponentPose=@n9, bAlphaBoolEnabled:bool=True, ConstraintWeights_0=1) -> (Pose) {
    alphaInputType=Bool
    alphaBoolBlend {blendInTime=0.025, blendOption=Linear, blendOutTime=0.25}
    boneToModify {boneName=dyn_skirt_fr_r}
    constraintSetup [{transformType=Rotation, targetBone {boneName=thigh_r}}]
    constraintWeights [1]
   }
   n11 = "Constraint - dyn_skirt_l (R)"!(ComponentPose=@n10, bAlphaBoolEnabled:bool=True, ConstraintWeights_0=1) -> (Pose) {
    alphaInputType=Bool
    alphaBoolBlend {blendInTime=0.025, blendOption=Linear, blendOutTime=0.25}
    boneToModify {boneName=dyn_skirt_l}
    constraintSetup [{transformType=Rotation, targetBone {boneName=thigh_l}}]
    constraintWeights [1]
   }
   n12 = "Constraint - dyn_skirt_r (R)"!(ComponentPose=@n11, bAlphaBoolEnabled:bool=True, ConstraintWeights_0=1) -> (Pose) {
    alphaInputType=Bool
    alphaBoolBlend {blendInTime=0.025, blendOption=Linear, blendOutTime=0.25}
    boneToModify {boneName=dyn_skirt_r}
    constraintSetup [{transformType=Rotation, targetBone {boneName=thigh_r}}]
    constraintWeights [1]
   }
   Output Pose(Result=Component To Local(ComponentPose=@n12)) {layerGroup=None}
   n13 = "Constraint - dyn_skirt_fr_mid_MSTR (PP)"!(
    ComponentPose=Local To Component(LocalPose=Copy Pose From Mesh() {bUseAttachedParent=true})
    Alpha=1
    ConstraintWeights_0=0.5
    ConstraintWeights_1:float=0.5
   ) -> (Pose) {
    alphaBoolBlend {blendOption=Linear}
    boneToModify {boneName=dyn_skirt_fr_mid_MSTR}
    constraintSetup [{targetBone {boneName=dyn_skirt_fr_l_MSTR}}, {targetBone {boneName=dyn_skirt_fr_r_MSTR}}]
    constraintWeights [1,1]
   }
   n14 = "Constraint - dyn_skirt_bk_mid_MSTR (PP)"!(ComponentPose=@n13, Alpha=1, ConstraintWeights_0=0.5, ConstraintWeights_1:float=0.5) -> (Pose) {
    alphaBoolBlend {blendOption=Linear}
    boneToModify {boneName=dyn_skirt_bk_mid_MSTR}
    constraintSetup [{targetBone {boneName=dyn_skirt_bk_l_MSTR}}, {targetBone {boneName=dyn_skirt_bk_r_MSTR}}]
    constraintWeights [1,1]
   }
   n15 = "Transform (Modify) Bone - Bone: dyn_skirt_fr_l_MSTR"(ComponentPose=@n14, Alpha=Get Left Leg Up Alpha() -> (Left Leg Up Alpha:double), Translation="0,-1,0", Rotation="0,0,0", Scale="1,1,1") {
    rotationMode=BMM_Additive
    rotationSpace=BCS_BoneSpace
    translationMode=BMM_Additive
    translationSpace=BCS_BoneSpace
    alphaBoolBlend {blendOption=Linear}
    boneToModify {boneName=dyn_skirt_fr_l_MSTR}
    translation {y=-1}
   }
   n16 = "Transform (Modify) Bone - Bone: dyn_skirt_fr_r_MSTR"(ComponentPose=@n15, Alpha=Get Right Leg Up Alpha() -> (Right Leg Up Alpha:double), Translation="0,-1,0", Rotation="0,0,0", Scale="1,1,1") {
    rotationMode=BMM_Additive
    rotationSpace=BCS_BoneSpace
    translationMode=BMM_Additive
    translationSpace=BCS_BoneSpace
    alphaBoolBlend {blendOption=Linear}
    boneToModify {boneName=dyn_skirt_fr_r_MSTR}
    translation {y=-1}
   }
   n17 = "Transform (Modify) Bone - Bone: dyn_skirt_l_MSTR"(ComponentPose=@n16, Alpha=Get Left Leg Out Alpha() -> (Left Leg Out Alpha:double), Translation="0,-3,0", Rotation="0,0,0", Scale="1,1,1") {
    rotationMode=BMM_Additive
    rotationSpace=BCS_BoneSpace
    translationMode=BMM_Additive
    translationSpace=BCS_BoneSpace
    alphaBoolBlend {blendOption=Linear}
    boneToModify {boneName=dyn_skirt_l_MSTR}
    translation {y=-3}
   }
   n18 = "Transform (Modify) Bone - Bone: dyn_skirt_r_MSTR"(ComponentPose=@n17, Alpha=Get Right Leg Out Alpha() -> (Right Leg Out Alpha:double), Translation="0,-3,0", Rotation="0,0,0", Scale="1,1,1") {
    rotationMode=BMM_Additive
    rotationSpace=BCS_BoneSpace
    translationMode=BMM_Additive
    translationSpace=BCS_BoneSpace
    alphaBoolBlend {blendOption=Linear}
    boneToModify {boneName=dyn_skirt_r_MSTR}
    translation {y=-3}
   }
   Save cached pose 'PreSim Cache'(Pose=Component To Local(ComponentPose=@n18)) {cacheName=PreSim Cache}
  ]}
 EventGraph {nodes [
   Event BlueprintUpdateAnimation()
   > n1 = Sequence!(execute) -> (then_0, then_1, then_2:Exec, then_3:Exec)
   > n2 = Gravity Override State Function DEPRECATED(CharacterPartAnimInstance=Self-Reference(), GameTimeInSeconds=GetGameTimeInSeconds(), GravityOverrideParams=Get Gravity Override Params())
   > Set GravityOverride(GravityOverride=@n2)
   n3 = Get bIsCrouching()
   n4 = Get bIsMoving2D()
   n5 = Get bIsSprinting()
   n6 = Get bIsOnGround() -> (bIsOnGround:bool)
   Set IsSkirtConstrain(
    execute=@n1.then_3
    IsSkirtConstrain:bool=OR Boolean!(A=AND Boolean(A=@n5, B=@n6), B=AND Boolean(A=@n3, B=@n4), C:bool=Get bIsPlayingEmote()) -> (ReturnValue)
   ) -> (Output_Get:bool)
   n7 = GetOwnerSkeletalMeshComponent()
   n8 = BreakRotator(InRot=BreakTransform(InTransform=GetDeltaTransformFromRefPose(self=@n7, BoneName=thigh_l, BaseName=pelvis)).Rotation)
   n9 = MapRangeClamped(Value=@n8.Yaw, InRangeA=-20, InRangeB=-80)
   Set Left Leg Up Alpha(execute=@n1.then_1, Left Leg Up Alpha:double=@n9) -> (Output_Get:double)
   > Set Left Leg Out Alpha(Left Leg Out Alpha:double=SelectFloat(A=0, B=MapRangeClamped(Value=@n8.Pitch, InRangeA=-10, InRangeB=30, OutRangeA=1, OutRangeB=0), bPickA=Get bIsSprinting())) -> (Output_Get:double)
   n10 = GetOwnerSkeletalMeshComponent()
   n11 = BreakRotator(InRot=BreakTransform(InTransform=GetDeltaTransformFromRefPose(self=@n10, BoneName=thigh_r, BaseName=pelvis)).Rotation)
   n12 = MapRangeClamped(Value=@n11.Yaw, InRangeA=-20, InRangeB=-80)
   > Set Right Leg Up Alpha(Right Leg Up Alpha:double=@n12) -> (Output_Get:double)
   > Set Right Leg Out Alpha(Right Leg Out Alpha:double=SelectFloat(A=0, B=MapRangeClamped(Value=@n11.Pitch, InRangeA=-10, InRangeB=30, OutRangeA=1, OutRangeB=0), bPickA=Get bIsSprinting())) -> (Output_Get:double)
   > Set Hoodie Center Alpha(Hoodie Center Alpha:double=SelectFloat(A=@n9, B=@n12, bPickA="float > float"(A=@n9, B=@n12))) -> (Output_Get:double)
   n13 = GetOwnerSkeletalMeshComponent()
   n14 = BreakRotator(InRot=BreakTransform(InTransform=GetDeltaTransformFromRefPose(self=@n13, BoneName=spine_05, BaseName=root)).Rotation)
   n15 = Get bIsCrouching()
   n16 = Get bIsPlayingEmote()
   Set Spine Rot Fwd Alpha(
    execute=@n1.then_2
    Spine Rot Fwd Alpha:double=SelectFloat(A=MapRangeClamped(Value=@n14.Yaw, InRangeA=15, InRangeB=50), B=0, bPickA=OR Boolean!(A=@n15, B=@n16) -> (ReturnValue))
   ) -> (Output_Get:double)
  ]}
]
variables [
 Gravity Override Params {
  type=GravityOverrideParamsStruct
  default {
   emoteOrMelee {gravityOverride=<0 0 -800>}
   front_End {gravityOverride=<0 0 -800>}
   onGround_Falling {gravityOverride=<0 0 -800>}
   onGround_Moving {gravityOverride=<0 0 -800>}
   onGround_Standing {gravityOverride=<0 0 -800>}
   ziplining {gravityOverride=<0 0 -800>}
  }
 }
 GravityOverride {type=Vector}
 Hoodie Center Alpha {type=number}
 IsSkirtConstrain {type=boolean}
 Left Leg Out Alpha {type=number}
 Left Leg Up Alpha {type=number}
 Right Leg Out Alpha {type=number}
 Right Leg Up Alpha {type=number}
 Spine Rot Fwd Alpha {type=number}
]
