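# Physics asset for the rear hair of the Nightingale female medium body's face accessory mesh, used for secondary hair motion.
# Capsules on head, neck, upper spine, clavicles and upper arms form the body shape, joined by angularly free constraints.
# Left, middle and right four-segment hair chains hang from master bones on the head, with swing limits widening toward the tips.
# Cross-constraints tie both side chains to the middle at segment 02 (limited) and 04 (free angular).
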PhysicsAsset /BRCosmetics/Characters/Player/Female/Medium/Bodies/F_MED_Nightingale/Meshes/Parts/F_MED_Nightingale_FaceAcc_Physics_rear
bodies {
 clavicle_l [{shapeType=Capsule, length=25, radius=4.5, shapeName=clavicle_l_capsule, center=<14 -1 0.5>, rotation=<rot 90 0 0>}]
 clavicle_r [{shapeType=Capsule, length=25, radius=4.5, shapeName=clavicle_r_capsule, center=<-14 1 -0.5>, rotation=<rot 90 0 0>}]
 dyn_hair_bk_l_01 [{shapeType=Capsule, length=4, radius=4, shapeName=dyn_hair_bk_mid_01_capsule, center=<5 -3 0>, rotation=<rot 90 0 0>}]
 dyn_hair_bk_l_02 [{shapeType=Capsule, length=4, radius=4, shapeName=dyn_hair_bk_mid_02_capsule, center=<5 -3.5 0>, rotation=<rot 90 0 0>}]
 dyn_hair_bk_l_03 [{shapeType=Capsule, length=4, radius=4.5, shapeName=dyn_hair_bk_mid_03_capsule, center=<5 -3.5 -1>, rotation=<rot 90 0 0>}]
 dyn_hair_bk_l_04 [{shapeType=Capsule, length=6, radius=4.5, shapeName=dyn_hair_bk_mid_04_capsule, center=<7 -3 -1>, rotation=<rot 90 0 0>}]
 dyn_hair_bk_l_MSTR [{shapeType=Capsule, length=1, radius=1, shapeName=dyn_hair_bk_l_MSTR_capsule}]
 dyn_hair_bk_mid_01 [{shapeType=Capsule, length=4, radius=4, shapeName=dyn_hair_bk_mid_01_capsule, center=<5 -3 0>, rotation=<rot 90 0 0>}]
 dyn_hair_bk_mid_02 [{shapeType=Capsule, length=4, radius=4, shapeName=dyn_hair_bk_mid_02_capsule, center=<5 -3 0>, rotation=<rot 90 0 0>}]
 dyn_hair_bk_mid_03 [{shapeType=Capsule, length=4, radius=4.5, shapeName=dyn_hair_bk_mid_03_capsule, center=<5 -2.5 0>, rotation=<rot 90 0 0>}]
 dyn_hair_bk_mid_04 [{shapeType=Capsule, length=6, radius=4.5, shapeName=dyn_hair_bk_mid_04_capsule, center=<7 -2 0>, rotation=<rot 90 0 0>}]
 dyn_hair_bk_mid_MSTR [{shapeType=Capsule, length=1, radius=1, shapeName=dyn_hair_bk_mid_MSTR_capsule}]
 dyn_hair_bk_r_01 [{shapeType=Capsule, length=4, radius=4, shapeName=dyn_hair_bk_mid_01_capsule, center=<5 -3 0>, rotation=<rot 90 0 0>}]
 dyn_hair_bk_r_02 [{shapeType=Capsule, length=4, radius=4, shapeName=dyn_hair_bk_mid_02_capsule, center=<5 -3.5 0>, rotation=<rot 90 0 0>}]
 dyn_hair_bk_r_03 [{shapeType=Capsule, length=4, radius=4.5, shapeName=dyn_hair_bk_mid_03_capsule, center=<5 -3.5 1>, rotation=<rot 90 0 0>}]
 dyn_hair_bk_r_04 [{shapeType=Capsule, length=6, radius=4.5, shapeName=dyn_hair_bk_mid_04_capsule, center=<7 -3 1>, rotation=<rot 90 0 0>}]
 dyn_hair_bk_r_MSTR [{shapeType=Capsule, length=1, radius=1, shapeName=dyn_hair_bk_r_MSTR_capsule}]
 head [{shapeType=Capsule, length=4.375, radius=6.75, center=<2.1043 1 0>, rotation=<rot 90 0 0>}]
 neck_01 [{shapeType=Capsule, length=7, radius=4.875, shapeName=neck_01_capsule, center=<5 0 0>, rotation=<rot 90 0 0>}]
 neck_02 [{shapeType=Capsule, length=0.505, radius=0.505, shapeName=neck_02_capsule, rotation=<rot 0 -90 0>}]
 spine_05 [
  {shapeType=Capsule, length=25, radius=14, center=<-2 4.5 0>}
  {shapeType=Capsule, length=5, radius=12, shapeName=spine_05_capsule, center=<2.5 3 0>, rotation=<rot 0 10 90>}
 ]
 upperarm_l [{shapeType=Capsule, length=22, radius=6, shapeName=upperarm_l_capsule, center=<12 0 0>, rotation=<rot 90 -90 -90>}]
 upperarm_r [{shapeType=Capsule, length=22, radius=6, shapeName=upperarm_r_capsule, center=<-12 0 0>, rotation=<rot 90 -90 -90>}]
}
constraints [
 {
  bone1Name=head
  bone2Name=neck_02
  swing1LimitDegrees=45
  swing1Motion=Free
  swing2LimitDegrees=45
  swing2Motion=Free
  twistLimitDegrees=45
  twistMotion=Free
 }
 {
  bone1Name=neck_01
  bone2Name=spine_05
  swing1LimitDegrees=45
  swing1Motion=Free
  swing2LimitDegrees=45
  swing2Motion=Free
  twistLimitDegrees=45
  twistMotion=Free
 }
 {
  bone1Name=clavicle_l
  bone2Name=spine_05
  swing1LimitDegrees=45
  swing1Motion=Free
  swing2LimitDegrees=45
  swing2Motion=Free
  twistLimitDegrees=45
  twistMotion=Free
 }
 {
  bone1Name=clavicle_r
  bone2Name=spine_05
  swing1LimitDegrees=45
  swing1Motion=Free
  swing2LimitDegrees=45
  swing2Motion=Free
  twistLimitDegrees=45
  twistMotion=Free
 }
 {
  bone1Name=upperarm_r
  bone2Name=clavicle_r
  swing1LimitDegrees=45
  swing1Motion=Free
  swing2LimitDegrees=45
  swing2Motion=Free
  twistLimitDegrees=45
  twistMotion=Free
 }
 {
  bone1Name=upperarm_l
  bone2Name=clavicle_l
  swing1LimitDegrees=45
  swing1Motion=Free
  swing2LimitDegrees=45
  swing2Motion=Free
  twistLimitDegrees=45
  twistMotion=Free
 }
 {
  bone1Name=neck_02
  bone2Name=neck_01
  swing1LimitDegrees=45
  swing1Motion=Free
  swing2LimitDegrees=45
  swing2Motion=Free
  twistLimitDegrees=45
  twistMotion=Free
 }
 {
  bone1Name=dyn_hair_bk_mid_MSTR
  bone2Name=head
  swing1LimitDegrees=45
  swing1Motion=Free
  swing2LimitDegrees=45
  swing2Motion=Free
  twistLimitDegrees=45
  twistMotion=Free
 }
 {bone1Name=dyn_hair_bk_mid_01, bone2Name=dyn_hair_bk_mid_MSTR, swing1LimitDegrees=8, swing2LimitDegrees=5, twistLimitDegrees=5}
 {bone1Name=dyn_hair_bk_mid_02, bone2Name=dyn_hair_bk_mid_01, swing1LimitDegrees=25, swing2LimitDegrees=15, twistLimitDegrees=5}
 {bone1Name=dyn_hair_bk_mid_03, bone2Name=dyn_hair_bk_mid_02, swing1LimitDegrees=30, swing2LimitDegrees=15, twistLimitDegrees=5}
 {bone1Name=dyn_hair_bk_mid_04, bone2Name=dyn_hair_bk_mid_03, swing1LimitDegrees=40, swing2LimitDegrees=20, twistLimitDegrees=5}
 {
  bone1Name=dyn_hair_bk_r_MSTR
  bone2Name=head
  swing1LimitDegrees=45
  swing1Motion=Free
  swing2LimitDegrees=45
  swing2Motion=Free
  twistLimitDegrees=45
  twistMotion=Free
 }
 {bone1Name=dyn_hair_bk_r_01, bone2Name=dyn_hair_bk_r_MSTR, swing1LimitDegrees=8, swing2LimitDegrees=5, twistLimitDegrees=5}
 {bone1Name=dyn_hair_bk_r_02, bone2Name=dyn_hair_bk_r_01, swing1LimitDegrees=25, swing2LimitDegrees=15, twistLimitDegrees=5}
 {bone1Name=dyn_hair_bk_r_03, bone2Name=dyn_hair_bk_r_02, swing1LimitDegrees=30, swing2LimitDegrees=15, twistLimitDegrees=5}
 {
  bone1Name=dyn_hair_bk_l_MSTR
  bone2Name=head
  swing1LimitDegrees=45
  swing1Motion=Free
  swing2LimitDegrees=45
  swing2Motion=Free
  twistLimitDegrees=45
  twistMotion=Free
 }
 {bone1Name=dyn_hair_bk_l_01, bone2Name=dyn_hair_bk_l_MSTR, swing1LimitDegrees=8, swing2LimitDegrees=5, twistLimitDegrees=5}
 {bone1Name=dyn_hair_bk_l_02, bone2Name=dyn_hair_bk_l_01, swing1LimitDegrees=25, swing2LimitDegrees=10, twistLimitDegrees=5}
 {bone1Name=dyn_hair_bk_l_03, bone2Name=dyn_hair_bk_l_02, swing1LimitDegrees=30, swing2LimitDegrees=15, twistLimitDegrees=5}
 {bone1Name=dyn_hair_bk_r_02, bone2Name=dyn_hair_bk_mid_02, swing1LimitDegrees=15, swing2LimitDegrees=10, twistLimitDegrees=10}
 {bone1Name=dyn_hair_bk_l_02, bone2Name=dyn_hair_bk_mid_02, swing1LimitDegrees=15, swing2LimitDegrees=10, twistLimitDegrees=10}
 {bone1Name=dyn_hair_bk_l_04, bone2Name=dyn_hair_bk_l_03, swing1LimitDegrees=40, swing2LimitDegrees=15, twistLimitDegrees=5}
 {bone1Name=dyn_hair_bk_r_04, bone2Name=dyn_hair_bk_r_03, swing1LimitDegrees=40, swing2LimitDegrees=15, twistLimitDegrees=5}
 {
  bone1Name=dyn_hair_bk_l_04
  bone2Name=dyn_hair_bk_mid_04
  swing1LimitDegrees=45
  swing1Motion=Free
  swing2LimitDegrees=45
  swing2Motion=Free
  twistLimitDegrees=45
  twistMotion=Free
 }
 {
  bone1Name=dyn_hair_bk_r_04
  bone2Name=dyn_hair_bk_mid_04
  swing1LimitDegrees=45
  swing1Motion=Free
  swing2LimitDegrees=45
  swing2Motion=Free
  twistLimitDegrees=45
  twistMotion=Free
 }
]
