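# Post-process style AnimBP for Nightingale's face accessory mesh, copying the parent pose and adding secondary motion to hair and earring.
# RigidBody nodes simulate bangs and rear hair; long non-cape backbling swaps rear simulation for bone constraints.
# With rigid bodies disabled, constraints and AnimDynamics on front strands substitute; the left earring always uses AnimDynamics.
# EventGraph derives corrective alphas from head and right-arm rotation, weapon type and melee state, and detects backbling once.
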AnimBlueprint /BRCosmetics/Characters/Player/Female/Medium/Bodies/F_MED_Nightingale/Meshes/Parts/F_MED_Nightingale_FaceAcc_AnimBP
target_skeleton=&/F_MED_Nightingale_FaceAcc_Skeleton
graphs [
 AnimGraph {nodes [
   n1 = "Transform (Modify) Bone - Bone: dyn_hair_fr_r_MSTR"(
    ComponentPose=Local To Component(LocalPose=Copy Pose From Mesh() {bUseAttachedParent=true})
    Alpha=Get Right Arm Up Alpha() -> (Right Arm Up Alpha:double)
    Translation="0,0,0"
    Rotation="17,-20,0"
    Scale="1,1,1"
   ) {
    rotationMode=BMM_Additive
    rotationSpace=BCS_BoneSpace
    translationMode=BMM_Additive
    translationSpace=BCS_BoneSpace
    alphaBoolBlend {blendOption=Linear}
    alphaScaleBiasClamp {bInterpResult=true}
    boneToModify {boneName=dyn_hair_fr_r_MSTR}
    rotation {pitch=17, yaw=-20}
   }
   n2 = Get Head Back Alpha() -> (Head Back Alpha:double)
   n3 = "Transform (Modify) Bone - Bone: dyn_hair_bk_mid_MSTR"(ComponentPose=@n1, Alpha=@n2, Translation="-1.5,0,0", Rotation="0,0,0", Scale="1,1,1") {
    rotationMode=BMM_Additive
    rotationSpace=BCS_BoneSpace
    translationMode=BMM_Additive
    translationSpace=BCS_BoneSpace
    alphaBoolBlend {blendOption=Linear}
    alphaScaleBiasClamp {bInterpResult=true}
    boneToModify {boneName=dyn_hair_bk_mid_MSTR}
    translation {x=-1.5}
   }
   n4 = "Transform (Modify) Bone - Bone: dyn_hair_bk_r_MSTR"(ComponentPose=@n3, Alpha=@n2, Translation="-1.5,0,0", Rotation="0,0,0", Scale="1,1,1") {
    rotationMode=BMM_Additive
    rotationSpace=BCS_BoneSpace
    translationMode=BMM_Additive
    translationSpace=BCS_BoneSpace
    alphaBoolBlend {blendOption=Linear}
    alphaScaleBiasClamp {bInterpResult=true}
    boneToModify {boneName=dyn_hair_bk_r_MSTR}
    translation {x=-1.5}
   }
   n5 = "Transform (Modify) Bone - Bone: dyn_hair_bk_l_MSTR"(ComponentPose=@n4, Alpha=@n2, Translation="-1.5,0,0", Rotation="0,0,0", Scale="1,1,1") {
    rotationMode=BMM_Additive
    rotationSpace=BCS_BoneSpace
    translationMode=BMM_Additive
    translationSpace=BCS_BoneSpace
    alphaBoolBlend {blendOption=Linear}
    alphaScaleBiasClamp {bInterpResult=true}
    boneToModify {boneName=dyn_hair_bk_l_MSTR}
    translation {x=-1.5}
   }
   n6 = Get Head Tilt Alpha() -> (Head Tilt Alpha:double)
   n7 = "Transform (Modify) Bone - Bone: dyn_hair_bk_mid_MSTR"(ComponentPose=@n5, Alpha=@n6, Translation="-1,0.5,0", Rotation="0,0,0", Scale="1,1,1") {
    rotationMode=BMM_Additive
    rotationSpace=BCS_BoneSpace
    translationMode=BMM_Additive
    translationSpace=BCS_BoneSpace
    alphaBoolBlend {blendOption=Linear}
    alphaScaleBiasClamp {bInterpResult=true}
    boneToModify {boneName=dyn_hair_bk_mid_MSTR}
    translation=<-1 0.5>
   }
   n8 = "Transform (Modify) Bone - Bone: dyn_hair_bk_r_MSTR"(ComponentPose=@n7, Alpha=@n6, Translation="-1,1,0", Rotation="0,0,0", Scale="1,1,1") {
    rotationMode=BMM_Additive
    rotationSpace=BCS_BoneSpace
    translationMode=BMM_Additive
    translationSpace=BCS_BoneSpace
    alphaBoolBlend {blendOption=Linear}
    alphaScaleBiasClamp {bInterpResult=true}
    boneToModify {boneName=dyn_hair_bk_r_MSTR}
    translation=<-1 1>
   }
   n9 = RigidBody!(
    ComponentPose=@n8
    Alpha=Get Melee Hair RBAN Alpha() -> (Melee Hair RBAN Alpha:double)
    OverrideWorldGravity:Vector="0,0,-800"
    ExternalForce
    ComponentLinearAccScale:Vector="0.25,0.25,0.45"
   ) -> (Pose) {
    bClampLinearTranslationLimitToRefPose=true
    bOverrideWorldGravity=true
    overridePhysicsAsset=&/F_MED_Nightingale_FaceAcc_Physics_bangs
    simulationSpace=BaseBoneSpace
    alphaBoolBlend {blendOption=Linear}
    alphaScaleBiasClamp {bInterpResult=true, interpSpeedIncreasing=2}
    baseBoneRef {boneName=neck_02}
    componentAppliedLinearAccClamp=<3000 3000 3000>
    componentLinearAccScale=<0.25 0.25 0.45>
    componentLinearVelScale=<0.25 0.25 0.35>
    overrideWorldGravity {z=-800}
   }
   n10 = RigidBody(ComponentPose=@n9, OverrideWorldGravity="0,0,-800", ComponentLinearAccScale="0.25,0.25,0.3") {
    bClampLinearTranslationLimitToRefPose=true
    bOverrideWorldGravity=true
    overridePhysicsAsset=&/F_MED_Nightingale_FaceAcc_Physics_rear
    simulationSpace=BaseBoneSpace
    alphaBoolBlend {blendOption=Linear}
    baseBoneRef {boneName=neck_02}
    componentAppliedLinearAccClamp=<3000 3000 3000>
    componentLinearAccScale=<0.25 0.25 0.3>
    componentLinearVelScale=<0.25 0.25 0.3>
    overrideWorldGravity {z=-800}
   }
   n11 = Get Right Arm Up Alpha()
   n12 = "Constraint - dyn_hair_r_02 (P)"(ComponentPose=@n10, Alpha=@n11, ConstraintWeights_0=1) {
    alphaBoolBlend {blendOption=Linear}
    alphaScaleBiasClamp {bClampResult=true, bInterpResult=true, bMapRange=true, clampMax=0.75}
    boneToModify {boneName=dyn_hair_r_02}
    constraintSetup [{targetBone {boneName=spine_05}}]
    constraintWeights [1]
   }
   n13 = "Constraint - dyn_hair_r_03 (P)"(ComponentPose=@n12, Alpha=@n11, ConstraintWeights_0=1) {
    alphaBoolBlend {blendOption=Linear}
    alphaScaleBiasClamp {bClampResult=true, bInterpResult=true, bMapRange=true, clampMax=0.8, inRange {max=1.25}}
    boneToModify {boneName=dyn_hair_r_03}
    constraintSetup [{targetBone {boneName=clavicle_r}}]
    constraintWeights [1]
   }
   n14 = Get Head Tilt Alpha()
   n15 = "Transform (Modify) Bone - Bone: dyn_hair_bk_r_MSTR"(ComponentPose=@n13, Alpha=@n14, Translation="0,0,0", Rotation="0,0,0", Scale="0.8,1,1") {
    scaleMode=BMM_Additive
    scaleSpace=BCS_BoneSpace
    alphaBoolBlend {blendOption=Linear}
    alphaScaleBiasClamp {bInterpResult=true}
    boneToModify {boneName=dyn_hair_bk_r_MSTR}
    scale {x=0.8}
   }
   n16 = "Transform (Modify) Bone - Bone: dyn_hair_bk_mid_MSTR"(ComponentPose=@n15, Alpha=@n14, Translation="0,0,0", Rotation="0,0,0", Scale="0.9,1,1") {
    scaleMode=BMM_Additive
    scaleSpace=BCS_BoneSpace
    alphaBoolBlend {blendOption=Linear}
    alphaScaleBiasClamp {bInterpResult=true}
    boneToModify {boneName=dyn_hair_bk_mid_MSTR}
    scale {x=0.9}
   }
   n17 = Get bIsPlayingMeleeAnim()
   n18 = "Constraint - dyn_hair_r_01 (P)"!(ComponentPose=@n16, bAlphaBoolEnabled:bool=@n17, ConstraintWeights_0=1) -> (Pose) {
    alphaInputType=Bool
    alphaBoolBlend {blendInTime=0.15, blendOption=Linear, blendOutTime=0.7}
    boneToModify {boneName=dyn_hair_r_01}
    constraintSetup [{targetBone {boneName=head}}]
    constraintWeights [1]
   }
   n19 = "Constraint - dyn_hair_r_02 (P)"!(ComponentPose=@n18, bAlphaBoolEnabled:bool=@n17, ConstraintWeights_0=1) -> (Pose) {
    alphaInputType=Bool
    alphaBoolBlend {blendInTime=0.15, blendOption=Linear, blendOutTime=0.7}
    boneToModify {boneName=dyn_hair_r_02}
    constraintSetup [{targetBone {boneName=neck_02}}]
    constraintWeights [1]
   }
   n20 = "Constraint - dyn_hair_r_03 (P)"!(ComponentPose=@n19, bAlphaBoolEnabled:bool=@n17, ConstraintWeights_0=1) -> (Pose) {
    alphaInputType=Bool
    alphaBoolBlend {blendInTime=0.15, blendOption=Linear, blendOutTime=0.7}
    boneToModify {boneName=dyn_hair_r_03}
    constraintSetup [{targetBone {boneName=clavicle_r}}]
    constraintWeights [1]
   }
   n21 = Get Melee Antenna Const Alpha() -> (Melee Antenna Const Alpha:double)
   n22 = "Constraint - dyn_hair_fr_r_02 (P)"(ComponentPose=@n20, Alpha=@n21, ConstraintWeights_0=1) {
    alphaBoolBlend {blendInTime=0.2, blendOption=Linear, blendOutTime=0.6}
    alphaScaleBiasClamp {bInterpResult=true, interpSpeedDecreasing=2}
    boneToModify {boneName=dyn_hair_fr_r_02}
    constraintSetup [{targetBone {boneName=neck_02}}]
    constraintWeights [1]
   }
   n23 = "Constraint - dyn_hair_fr_r_03 (P)"(ComponentPose=@n22, Alpha=@n21, ConstraintWeights_0=1) {
    alphaBoolBlend {blendInTime=0.2, blendOption=Linear, blendOutTime=0.6}
    alphaScaleBiasClamp {bInterpResult=true, interpSpeedDecreasing=2}
    boneToModify {boneName=dyn_hair_fr_r_03}
    constraintSetup [{targetBone {boneName=spine_05}}]
    constraintWeights [1]
   }
   n24 = "Constraint - dyn_hair_fr_r_04 (P)"(ComponentPose=@n23, Alpha=@n21, ConstraintWeights_0=1) {
    alphaBoolBlend {blendInTime=0.2, blendOption=Linear, blendOutTime=0.6}
    alphaScaleBiasClamp {bInterpResult=true, interpSpeedDecreasing=2}
    boneToModify {boneName=dyn_hair_fr_r_04}
    constraintSetup [{targetBone {boneName=clavicle_r}}]
    constraintWeights [1]
   }
   n25 = "Transform (Modify) Bone - Bone: dyn_hair_bk_l_MSTR"!(ComponentPose=@n24, bAlphaBoolEnabled:bool=True, Translation="-1,0,0", Rotation="0,-3,0", Scale="1,1,1") -> (Pose) {
    alphaInputType=Bool
    rotationMode=BMM_Additive
    rotationSpace=BCS_BoneSpace
    translationMode=BMM_Additive
    translationSpace=BCS_BoneSpace
    alphaBoolBlend {blendOption=Linear}
    boneToModify {boneName=dyn_hair_bk_l_MSTR}
    rotation {yaw=-3}
    translation {x=-1}
   }
   n26 = "Transform (Modify) Bone - Bone: dyn_hair_bk_r_MSTR"!(ComponentPose=@n25, bAlphaBoolEnabled:bool=True, Translation="0,0,0", Rotation="-3,0,0", Scale="1,1,1") -> (Pose) {
    alphaInputType=Bool
    rotationMode=BMM_Additive
    rotationSpace=BCS_BoneSpace
    translationMode=BMM_Additive
    translationSpace=BCS_BoneSpace
    alphaBoolBlend {blendOption=Linear}
    boneToModify {boneName=dyn_hair_bk_r_MSTR}
    rotation {pitch=-3}
   }
   n27 = "Constraint - dyn_hair_bk_mid_02 (P)"(ComponentPose=Local To Component(LocalPose=Copy Pose From Mesh() {bUseAttachedParent=true}), Alpha=1, ConstraintWeights_0=1) {
    alphaBoolBlend {blendOption=Linear}
    boneToModify {boneName=dyn_hair_bk_mid_02}
    constraintSetup [{targetBone {boneName=neck_02}}]
    constraintWeights [1]
   }
   n28 = "Constraint - dyn_hair_bk_r_02 (PP)"!(ComponentPose=@n27, Alpha=1, ConstraintWeights_0=0.5, ConstraintWeights_1:float=0.5) -> (Pose) {
    alphaBoolBlend {blendOption=Linear}
    boneToModify {boneName=dyn_hair_bk_r_02}
    constraintSetup [{targetBone {boneName=neck_02}}, {targetBone {boneName=spine_05}}]
    constraintWeights [1,1]
   }
   n29 = "Constraint - dyn_hair_bk_l_02 (PP)"!(ComponentPose=@n28, Alpha=1, ConstraintWeights_0=0.5, ConstraintWeights_1:float=0.5) -> (Pose) {
    alphaBoolBlend {blendOption=Linear}
    boneToModify {boneName=dyn_hair_bk_l_02}
    constraintSetup [{targetBone {boneName=neck_02}}, {targetBone {boneName=spine_05}}]
    constraintWeights [1,1]
   }
   n30 = "Constraint - dyn_hair_bk_mid_03 (P)"(ComponentPose=@n29, Alpha=1, ConstraintWeights_0=1) {
    alphaBoolBlend {blendOption=Linear}
    boneToModify {boneName=dyn_hair_bk_mid_03}
    constraintSetup [{targetBone {boneName=spine_05}}]
    constraintWeights [1]
   }
   n31 = "Constraint - dyn_hair_bk_r_03 (PP)"!(ComponentPose=@n30, Alpha=1, ConstraintWeights_0=0.5, ConstraintWeights_1:float=1) -> (Pose) {
    alphaBoolBlend {blendOption=Linear}
    boneToModify {boneName=dyn_hair_bk_r_03}
    constraintSetup [{targetBone {boneName=spine_05}}, {targetBone {boneName=clavicle_r}}]
    constraintWeights [1,1]
   }
   n32 = "Constraint - dyn_hair_bk_l_03 (PP)"!(ComponentPose=@n31, Alpha=1, ConstraintWeights_0=0.5, ConstraintWeights_1:float=0.5) -> (Pose) {
    alphaBoolBlend {blendOption=Linear}
    boneToModify {boneName=dyn_hair_bk_l_03}
    constraintSetup [{targetBone {boneName=spine_05}}, {targetBone {boneName=clavicle_l}}]
    constraintWeights [1,1]
   }
   n33 = "Constraint - dyn_hair_bk_mid_04 (P)"(ComponentPose=@n32, Alpha=1, ConstraintWeights_0=1) {
    alphaBoolBlend {blendOption=Linear}
    boneToModify {boneName=dyn_hair_bk_mid_04}
    constraintSetup [{targetBone {boneName=spine_05}}]
    constraintWeights [1]
   }
   n34 = "Constraint - dyn_hair_bk_r_04 (PP)"!(ComponentPose=@n33, Alpha=1, ConstraintWeights_0=0, ConstraintWeights_1:float=1) -> (Pose) {
    alphaBoolBlend {blendOption=Linear}
    boneToModify {boneName=dyn_hair_bk_r_04}
    constraintSetup [{targetBone {boneName=spine_05}}, {targetBone {boneName=clavicle_r}}]
    constraintWeights [1,1]
   }
   n35 = "Constraint - dyn_hair_bk_l_04 (PP)"!(ComponentPose=@n34, Alpha=1, ConstraintWeights_0=0, ConstraintWeights_1:float=1) -> (Pose) {
    alphaBoolBlend {blendOption=Linear}
    boneToModify {boneName=dyn_hair_bk_l_04}
    constraintSetup [{targetBone {boneName=spine_05}}, {targetBone {boneName=clavicle_l}}]
    constraintWeights [1,1]
   }
   n36 = Get Right Arm Up Alpha()
   n37 = "Transform (Modify) Bone - Bone: dyn_hair_fr_r_MSTR"(ComponentPose=@n35, Alpha=@n36, Translation="0,0,0", Rotation="17,-20,0", Scale="1,1,1") {
    rotationMode=BMM_Additive
    rotationSpace=BCS_BoneSpace
    translationMode=BMM_Additive
    translationSpace=BCS_BoneSpace
    alphaBoolBlend {blendOption=Linear}
    boneToModify {boneName=dyn_hair_fr_r_MSTR}
    rotation {pitch=17, yaw=-20}
   }
   n38 = "Transform (Modify) Bone - Bone: dyn_hair_r_MSTR"(ComponentPose=@n37, Alpha=@n36, Translation="-2,0,0", Rotation="0,5,0", Scale="1,1,1") {
    rotationMode=BMM_Additive
    rotationSpace=BCS_BoneSpace
    translationMode=BMM_Additive
    translationSpace=BCS_BoneSpace
    alphaBoolBlend {blendOption=Linear}
    boneToModify {boneName=dyn_hair_r_MSTR}
    rotation {yaw=5}
    translation {x=-2}
   }
   n39 = RigidBody(ComponentPose=@n38, Alpha=Get Melee Hair RBAN Alpha(), OverrideWorldGravity="0,0,-800", ComponentLinearAccScale="0.25,0.25,0.45") {
    bClampLinearTranslationLimitToRefPose=true
    bOverrideWorldGravity=true
    overridePhysicsAsset=&F_MED_Nightingale_FaceAcc_Physics_bangs
    simulationSpace=BaseBoneSpace
    alphaBoolBlend {blendOption=Linear}
    alphaScaleBiasClamp {bInterpResult=true, interpSpeedIncreasing=2}
    baseBoneRef {boneName=neck_02}
    componentAppliedLinearAccClamp=<5000 5000 5000>
    componentLinearAccScale=<0.25 0.25 0.45>
    componentLinearVelScale=<0.25 0.25 0.35>
    overrideWorldGravity {z=-800}
   }
   n40 = Get Right Arm Up Alpha()
   n41 = "Constraint - dyn_hair_r_02 (PP)"!(ComponentPose=@n39, Alpha=@n40, ConstraintWeights_0=0.5, ConstraintWeights_1:float=0.5) -> (Pose) {
    alphaBoolBlend {blendOption=Linear}
    boneToModify {boneName=dyn_hair_r_02}
    constraintSetup [{targetBone {boneName=dyn_hair_r_01}}, {targetBone {boneName=clavicle_r}}]
    constraintWeights [1,1]
   }
   n42 = "Constraint - dyn_hair_r_03 (PP)"!(ComponentPose=@n41, Alpha=@n40, ConstraintWeights_0=0.25, ConstraintWeights_1:float=0.75) -> (Pose) {
    alphaBoolBlend {blendOption=Linear}
    boneToModify {boneName=dyn_hair_r_03}
    constraintSetup [{targetBone {boneName=spine_05}}, {targetBone {boneName=clavicle_r}}]
    constraintWeights [1,1]
   }
   n43 = Get bIsPlayingMeleeAnim()
   n44 = "Constraint - dyn_hair_r_01 (P)"(ComponentPose=@n42, bAlphaBoolEnabled=@n43, ConstraintWeights_0=1) {
    alphaInputType=Bool
    alphaBoolBlend {blendInTime=0.15, blendOption=Linear, blendOutTime=0.7}
    boneToModify {boneName=dyn_hair_r_01}
    constraintSetup [{targetBone {boneName=head}}]
    constraintWeights [1]
   }
   n45 = "Constraint - dyn_hair_r_02 (P)"(ComponentPose=@n44, bAlphaBoolEnabled=@n43, ConstraintWeights_0=1) {
    alphaInputType=Bool
    alphaBoolBlend {blendInTime=0.15, blendOption=Linear, blendOutTime=0.7}
    boneToModify {boneName=dyn_hair_r_02}
    constraintSetup [{targetBone {boneName=neck_02}}]
    constraintWeights [1]
   }
   n46 = "Constraint - dyn_hair_r_03 (P)"(ComponentPose=@n45, bAlphaBoolEnabled=@n43, ConstraintWeights_0=1) {
    alphaInputType=Bool
    alphaBoolBlend {blendInTime=0.15, blendOption=Linear, blendOutTime=0.7}
    boneToModify {boneName=dyn_hair_r_03}
    constraintSetup [{targetBone {boneName=clavicle_r}}]
    constraintWeights [1]
   }
   n47 = Get Melee Antenna Const Alpha()
   n48 = "Constraint - dyn_hair_fr_r_02 (P)"(ComponentPose=@n46, Alpha=@n47, ConstraintWeights_0=1) {
    alphaBoolBlend {blendInTime=0.2, blendOption=Linear, blendOutTime=0.6}
    alphaScaleBiasClamp {bInterpResult=true, interpSpeedDecreasing=2}
    boneToModify {boneName=dyn_hair_fr_r_02}
    constraintSetup [{targetBone {boneName=neck_02}}]
    constraintWeights [1]
   }
   n49 = "Constraint - dyn_hair_fr_r_03 (P)"(ComponentPose=@n48, Alpha=@n47, ConstraintWeights_0=1) {
    alphaBoolBlend {blendInTime=0.2, blendOption=Linear, blendOutTime=0.6}
    alphaScaleBiasClamp {bInterpResult=true, interpSpeedDecreasing=2}
    boneToModify {boneName=dyn_hair_fr_r_03}
    constraintSetup [{targetBone {boneName=spine_05}}]
    constraintWeights [1]
   }
   n50 = "Constraint - dyn_hair_fr_r_04 (P)"(ComponentPose=@n49, Alpha=@n47, ConstraintWeights_0=1) {
    alphaBoolBlend {blendInTime=0.2, blendOption=Linear, blendOutTime=0.6}
    alphaScaleBiasClamp {bInterpResult=true, interpSpeedDecreasing=2}
    boneToModify {boneName=dyn_hair_fr_r_04}
    constraintSetup [{targetBone {boneName=clavicle_r}}]
    constraintWeights [1]
   }
   n51 = Blend Poses by bool(
    BlendPose_0=Component To Local(ComponentPose=@n50)
    BlendPose_1=Component To Local(ComponentPose=@n26)
    bActiveValue=Get IsBackblingOn() -> (IsBackblingOn:bool)
   ) {blendType=Linear}
   n52 = "Constraint - dyn_hair_bk_mid_02 (P)"(ComponentPose=Local To Component(LocalPose=Copy Pose From Mesh() {bUseAttachedParent=true}), Alpha=1, ConstraintWeights_0=1) {
    alphaBoolBlend {blendOption=Linear}
    boneToModify {boneName=dyn_hair_bk_mid_02}
    constraintSetup [{targetBone {boneName=neck_02}}]
    constraintWeights [1]
   }
   n53 = "Constraint - dyn_hair_bk_r_02 (PP)"(ComponentPose=@n52, Alpha=1, ConstraintWeights_0=0.5, ConstraintWeights_1=0.5) {
    alphaBoolBlend {blendOption=Linear}
    boneToModify {boneName=dyn_hair_bk_r_02}
    constraintSetup [{targetBone {boneName=neck_02}}, {targetBone {boneName=spine_05}}]
    constraintWeights [1,1]
   }
   n54 = "Constraint - dyn_hair_bk_l_02 (PP)"(ComponentPose=@n53, Alpha=1, ConstraintWeights_0=0.5, ConstraintWeights_1=0.5) {
    alphaBoolBlend {blendOption=Linear}
    boneToModify {boneName=dyn_hair_bk_l_02}
    constraintSetup [{targetBone {boneName=neck_02}}, {targetBone {boneName=spine_05}}]
    constraintWeights [1,1]
   }
   n55 = "Constraint - dyn_hair_bk_mid_03 (P)"(ComponentPose=@n54, Alpha=1, ConstraintWeights_0=1) {
    alphaBoolBlend {blendOption=Linear}
    boneToModify {boneName=dyn_hair_bk_mid_03}
    constraintSetup [{targetBone {boneName=spine_05}}]
    constraintWeights [1]
   }
   n56 = "Constraint - dyn_hair_bk_r_03 (PP)"(ComponentPose=@n55, Alpha=1, ConstraintWeights_0=0.5, ConstraintWeights_1=1) {
    alphaBoolBlend {blendOption=Linear}
    boneToModify {boneName=dyn_hair_bk_r_03}
    constraintSetup [{targetBone {boneName=spine_05}}, {targetBone {boneName=clavicle_r}}]
    constraintWeights [1,1]
   }
   n57 = "Constraint - dyn_hair_bk_l_03 (PP)"(ComponentPose=@n56, Alpha=1, ConstraintWeights_0=0.5, ConstraintWeights_1=0.5) {
    alphaBoolBlend {blendOption=Linear}
    boneToModify {boneName=dyn_hair_bk_l_03}
    constraintSetup [{targetBone {boneName=spine_05}}, {targetBone {boneName=clavicle_l}}]
    constraintWeights [1,1]
   }
   n58 = "Constraint - dyn_hair_bk_mid_04 (P)"(ComponentPose=@n57, Alpha=1, ConstraintWeights_0=1) {
    alphaBoolBlend {blendOption=Linear}
    boneToModify {boneName=dyn_hair_bk_mid_04}
    constraintSetup [{targetBone {boneName=spine_05}}]
    constraintWeights [1]
   }
   n59 = "Constraint - dyn_hair_bk_r_04 (PP)"(ComponentPose=@n58, Alpha=1, ConstraintWeights_0=0, ConstraintWeights_1=1) {
    alphaBoolBlend {blendOption=Linear}
    boneToModify {boneName=dyn_hair_bk_r_04}
    constraintSetup [{targetBone {boneName=spine_05}}, {targetBone {boneName=clavicle_r}}]
    constraintWeights [1,1]
   }
   n60 = "Constraint - dyn_hair_bk_l_04 (PP)"(ComponentPose=@n59, Alpha=1, ConstraintWeights_0=0, ConstraintWeights_1=1) {
    alphaBoolBlend {blendOption=Linear}
    boneToModify {boneName=dyn_hair_bk_l_04}
    constraintSetup [{targetBone {boneName=spine_05}}, {targetBone {boneName=clavicle_l}}]
    constraintWeights [1,1]
   }
   n61 = "Constraint - dyn_hair_r_02 (PP)"(ComponentPose=@n60, Alpha=1, ConstraintWeights_0=0.5, ConstraintWeights_1=0.5) {
    alphaBoolBlend {blendOption=Linear}
    boneToModify {boneName=dyn_hair_r_02}
    constraintSetup [{targetBone {boneName=head}}, {targetBone {boneName=clavicle_r}}]
    constraintWeights [1,1]
   }
   n62 = "Constraint - dyn_hair_r_03 (PP)"(ComponentPose=@n61, Alpha=1, ConstraintWeights_0=0.25, ConstraintWeights_1=0.75) {
    alphaBoolBlend {blendOption=Linear}
    boneToModify {boneName=dyn_hair_r_03}
    constraintSetup [{targetBone {boneName=head}}, {targetBone {boneName=clavicle_r}}]
    constraintWeights [1,1]
   }
   n63 = Anim Dynamics - Bone: dyn_hair_fr_r_02!(ComponentPose=@n62, Alpha=1, GravityOverride:Vector="0,0,-980", ExternalForce="0,0,0") -> (Pose) {
    angularDampingOverride=0.998
    bOverrideAngularDamping=true
    bUseGravityOverride=true
    simulationSpace=RootRelative
    alphaBoolBlend {blendOption=Linear}
    boundBone {boneName=dyn_hair_fr_r_02}
    componentAppliedLinearAccClamp=<3000 3000 3000>
    componentLinearAccScale=<0.2 0.2 0.3>
    componentLinearVelScale=<0.2 0.2 0.3>
    gravityOverride {z=-980}
    physicsBodyDefinitions [{
      boundBone {boneName=dyn_hair_fr_r_02}
      boxExtents=<15 10 10>
      constraintSetup {angularLimitsMax=<1 10 10>, angularLimitsMin=<-1 -20 -20>, angularTarget=<0 0 0>}
      localJointOffset=<3 0 0>
     }]
   }
   n64 = Anim Dynamics - Bone: dyn_hair_fr_l_02!(ComponentPose=@n63, Alpha=1, GravityOverride:Vector="0,0,-980", ExternalForce="0,0,0") -> (Pose) {
    angularDampingOverride=0.998
    bOverrideAngularDamping=true
    bUseGravityOverride=true
    simulationSpace=RootRelative
    alphaBoolBlend {blendOption=Linear}
    boundBone {boneName=dyn_hair_fr_l_02}
    componentAppliedLinearAccClamp=<3000 3000 3000>
    componentLinearAccScale=<0.2 0.2 0.3>
    componentLinearVelScale=<0.2 0.2 0.3>
    gravityOverride {z=-980}
    physicsBodyDefinitions [{
      boundBone {boneName=dyn_hair_fr_l_02}
      boxExtents=<15 10 10>
      constraintSetup {angularLimitsMax=<1 10 10>, angularLimitsMin=<-1 -20 -20>, angularTarget=<0 0 0>}
      localJointOffset=<3 0 0>
     }]
   }
   n65 = Anim Dynamics - Bone: dyn_hair_fr_r_03!(ComponentPose=@n64, Alpha=1, GravityOverride:Vector="0,0,-980", ExternalForce="0,0,0") -> (Pose) {
    angularDampingOverride=0.998
    bOverrideAngularDamping=true
    bUseGravityOverride=true
    simulationSpace=RootRelative
    alphaBoolBlend {blendOption=Linear}
    boundBone {boneName=dyn_hair_fr_r_03}
    componentAppliedLinearAccClamp=<3000 3000 3000>
    componentLinearAccScale=<0.2 0.2 0.3>
    componentLinearVelScale=<0.2 0.2 0.3>
    gravityOverride {z=-980}
    physicsBodyDefinitions [{
      boundBone {boneName=dyn_hair_fr_r_03}
      boxExtents=<15 10 10>
      constraintSetup {angularLimitsMax=<1 10 10>, angularLimitsMin=<-1 -20 -20>, angularTarget=<0 0 0>}
      localJointOffset=<3 0 0>
     }]
   }
   n66 = Anim Dynamics - Bone: dyn_hair_fr_l_03!(ComponentPose=@n65, Alpha=1, GravityOverride:Vector="0,0,-980", ExternalForce="0,0,0") -> (Pose) {
    angularDampingOverride=0.998
    bOverrideAngularDamping=true
    bUseGravityOverride=true
    simulationSpace=RootRelative
    alphaBoolBlend {blendOption=Linear}
    boundBone {boneName=dyn_hair_fr_l_03}
    componentAppliedLinearAccClamp=<3000 3000 3000>
    componentLinearAccScale=<0.2 0.2 0.3>
    componentLinearVelScale=<0.2 0.2 0.3>
    gravityOverride {z=-980}
    physicsBodyDefinitions [{
      boundBone {boneName=dyn_hair_fr_l_03}
      boxExtents=<15 10 10>
      constraintSetup {angularLimitsMax=<1 10 10>, angularLimitsMin=<-1 -20 -20>, angularTarget=<0 0 0>}
      localJointOffset=<3 0 0>
     }]
   }
   n67 = Component To Local(ComponentPose=@n66)
   n68 = Get bRigidBodyEnabled()
   n69 = Anim Dynamics - Bone: dyn_earing_l_01!(
    ComponentPose=Local To Component(LocalPose=Blend Poses by bool(BlendPose_0=@n51, BlendPose_1=@n67, bActiveValue=@n68) {blendType=Linear})
    Alpha=1
    GravityOverride:Vector="0,0,-800"
    ExternalForce="0,0,0"
   ) -> (Pose) {
    angularDampingOverride=0.99
    bOverrideAngularDamping=true
    bUseGravityOverride=true
    alphaBoolBlend {blendOption=Linear}
    boundBone {boneName=dyn_earing_l_01}
    componentAppliedLinearAccClamp=<3000 3000 3000>
    componentLinearAccScale=<0.1 0.1 0.2>
    componentLinearVelScale=<0.1 0.1 0.2>
    gravityOverride {z=-800}
    physicsBodyDefinitions [{
      boundBone {boneName=dyn_earing_l_01}
      boxExtents=<20 20 5>
      constraintSetup {angularLimitsMax=<2 5 20>, angularLimitsMin=<-2 -15 -15>, angularTarget=<0 0 0>}
      localJointOffset=<3 0 0>
     }]
   }
   Output Pose(Result=Component To Local(ComponentPose=@n69)) {layerGroup=None}
  ]}
 EventGraph {nodes [
   Event BlueprintUpdateAnimation()
   > n1 = Sequence!(execute) -> (then_0, then_1, then_2:Exec, then_3:Exec, then_4:Exec)
   > n2 = Gravity Override State Function DEPRECATED(CharacterPartAnimInstance=Self-Reference(), GameTimeInSeconds=GetGameTimeInSeconds(), GravityOverrideParams=Get Gravity Override Params())
   > Set GravityOverride(GravityOverride=@n2)
   n3 = Cast To FortPlayerAnimInstance(execute=@n1.then_3, Object=GetAnimInstance(self=GetOwnerSkeletalMeshComponent()))
   n4 = GetOwnerSkeletalMeshComponent()
   n5 = BreakRotator(InRot=BreakTransform(InTransform=GetDeltaTransformFromRefPose(self=@n4, BoneName=upperArm_r, BaseName=spine_05)).Rotation)
   > Set Right Arm Up Alpha(Right Arm Up Alpha:double=SelectFloat(A=0, B=MapRangeClamped(Value=@n5.Yaw, InRangeA=30, InRangeB=120, OutRangeB=0.3), bPickA=Get bIsPlayingEmote())) -> (Output_Get:double)
   > n6 = Switch on EFortWeaponCoreAnimation!(execute, Selection:EFortWeaponCoreAnimation Enum=Get WeaponCoreAnim(self=@n3)) -> (
    Melee:Exec
    Pistol:Exec
    Shotgun:Exec
    PaperBlueprint:Exec
    Rifle:Exec
    MeleeOneHand:Exec
    MachinePistol:Exec
    RocketLauncher:Exec
    GrenadeLauncher:Exec
    GoingCommando:Exec
    AssaultRifle:Exec
    TacticalShotgun:Exec
    SniperRifle:Exec
    TrapPlacement:Exec
    ShoulderLauncher:Exec
    AbilityDecoTool:Exec
    Crossbow:Exec
    C4:Exec
    RemoteControl:Exec
    DualWield:Exec
    AR_BullPup:Exec
    AR_ForwardGrip:Exec
    MedPackPaddles:Exec
    SMG_P90:Exec
    AR_DrumGun:Exec
    Consumable_Small:Exec
    Consumable_Large:Exec
    Balloon:Exec
    MountedTurret:Exec
    CreativeTool:Exec
    ExplosiveBow:Exec
    AshtonIndigo:Exec
    AshtonChicago:Exec
    MeleeDualWield:Exec
    Unarmed:Exec
   )
   Set Right Arm Up Alpha(execute=[@n6.DualWield, @n6.Melee], Right Arm Up Alpha=0)
   Set Right Arm Up Alpha(
    execute=[@n6.MachinePistol, @n6.Pistol]
    Right Arm Up Alpha=SelectFloat(A=0, B=MapRangeClamped(Value=@n5.Yaw, InRangeA=20, InRangeB=100), bPickA=Get bIsPlayingEmote())
   )
   Is Valid(exec=@n1.then_1, InputObject=Get Body Mesh() -> (Body Mesh:Skeletal Mesh Component Object Reference))
   > Do Once()
   n7 = CheckForBackbling(LongBackblingThreshold=40)
   > n8 = Branch(Condition=AND Boolean(A=@n7, B=NOT Boolean(A=@n7.bIsBackblingACape)))
   > Set IsBackblingOn(IsBackblingOn:bool=true) -> (Output_Get:bool)
   Set IsBackblingOn(execute=@n8.else, IsBackblingOn=false)
   n9 = Get bIsTargeting()
   n10 = Get bIsPlayingMeleeAnim()
   n11 = Get bIsSkyDiving()
   Set Hair Velocity(
    execute=@n1.then_4
    Hair Velocity:Vector=SelectVector(A="0.15,0.15,0.25", B="0.25,0.25,0.45", bPickA=OR Boolean!(A=@n9, B=@n10, C:bool=@n11) -> (ReturnValue))
   ) -> (Output_Get:Vector)
   > n12 = Branch(Condition=Get bIsPlayingMeleeAnim())
   > Set Melee Hair RBAN Alpha(Melee Hair RBAN Alpha:double=0.4) -> (Output_Get:double)
   > Set Melee Antenna Const Alpha(Melee Antenna Const Alpha:double=0.5) -> (Output_Get:double)
   Set Melee Hair RBAN Alpha(execute=@n12.else, Melee Hair RBAN Alpha=1)
   > Set Melee Antenna Const Alpha(Melee Antenna Const Alpha=0)
   n13 = Get bIsCrouching()
   n14 = GetOwnerSkeletalMeshComponent()
   n15 = BreakRotator(InRot=BreakTransform(InTransform=GetDeltaTransformFromRefPose(self=@n14, BoneName=head, BaseName=spine_05)).Rotation)
   Set Head Back Alpha(
    execute=@n1.then_2
    Head Back Alpha:double=MapRangeClamped(Value=@n15.Yaw, InRangeA=SelectFloat(A=-5, B=-5, bPickA=@n13), InRangeB=SelectFloat(A=-85, B=-45, bPickA=@n13))
   ) -> (Output_Get:double)
   > Set Head Tilt Alpha(Head Tilt Alpha:double=MapRangeClamped(Value=@n15.Pitch, InRangeA=0, InRangeB=30)) -> (Output_Get:double)
   Event BlueprintInitializeAnimation()
   > n16 = Cast To FortPlayerPawnAthena(Object=TryGetPawnOwner())
   > Set Body Mesh(Body Mesh:Skeletal Mesh Component Object Reference=GetSkeletalMeshForPartType(self=@n16)) -> (Output_Get:Skeletal Mesh Component Object Reference)
   Set Head Back Shldr Alpha(Head Back Shldr Alpha:double="Absolute (Float)"(A=float - float(B=1))) -> (Output_Get:double)
  ]}
]
variables [
 Body Mesh {type=SkeletalMeshComponent}
 Gravity Override Params {
  type=GravityOverrideParamsStruct
  default {
   emoteOrMelee {gravityOverride=<0 0 -800>}
   front_End {gravityOverride=<0 0 -800>}
   onGround_Falling {gravityOverride=<0 0 -800>}
   onGround_Moving {gravityOverride=<0 0 -800>}
   onGround_Standing {gravityOverride=<0 0 -800>}
   ziplining {gravityOverride=<0 0 -800>}
  }
 }
 GravityOverride {type=Vector}
 Hair Velocity {type=Vector}
 Head Back Alpha {type=number}
 Head Back Shldr Alpha {type=number}
 Head Tilt Alpha {type=number}
 IsBackblingOn {type=boolean}
 Melee Antenna Const Alpha {type=number}
 Melee Hair RBAN Alpha {type=number}
 Right Arm Up Alpha {type=number}
]
